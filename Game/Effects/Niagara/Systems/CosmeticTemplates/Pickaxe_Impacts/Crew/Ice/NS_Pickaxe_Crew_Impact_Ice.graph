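# Niagara system for the ice-themed Crew pickaxe cosmetic's impact effect.
# It combines six emitters: snowflakes, ice shards, smoke puff, star spike, spike streak, and a flare with light.
# System scripts run SystemState and a variable assignment, with dependencies on curve-driven floats, color construction and a time-of-day scalar input.
# User parameters expose front-end and front-end preview flags plus a swing section index.
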NiagaraSystem /Game/Effects/Niagara/Systems/CosmeticTemplates/Pickaxe_Impacts/Crew/Ice/NS_Pickaxe_Crew_Impact_Ice
dependencies {
 usedDataInterfaces [&NiagaraDataInterfaceCurve]
 usedDynamicInputs [
  &MakeLinearColorFromVectorAndFloat
  &VectorFromFloat
  &FNTimeOfDayScalar
  &FloatFromCurve
 ]
 usedModules [
  &self:SystemSpawnScript.SystemScriptSource.SystemScriptGraph.NiagaraNodeAssignment_1.SetVariables_#1
  &SystemState
 ]
}
emitters {
 FlareAndLight {rendererClasses [], renderers []}
 IceShard {rendererClasses [], renderers []}
 SmokePuff {rendererClasses [], renderers []}
 SnowFlakes {rendererClasses [], renderers []}
 SpikeStreak {rendererClasses [], renderers []}
 StarSpike {rendererClasses [], renderers []}
}
summary {
 systemName=NS_Pickaxe_Crew_Impact_Ice
 emitters [
  {emitterName=SnowFlakes, rendererClasses []}
  {emitterName=FlareAndLight, rendererClasses []}
  {emitterName=SpikeStreak, rendererClasses []}
  {emitterName=StarSpike, rendererClasses []}
  {emitterName=IceShard, rendererClasses []}
  {emitterName=SmokePuff, rendererClasses []}
 ]
 userVariables [
  User.bIsFrontEnd {type=NiagaraBool}
  User.bIsFrontEndPreview {type=NiagaraBool}
  User.SwingSectionIndex {type=NiagaraInt32}
 ]
}
